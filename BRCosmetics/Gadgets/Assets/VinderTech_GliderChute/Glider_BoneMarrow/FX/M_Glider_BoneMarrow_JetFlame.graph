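# Translucent, unlit, two-sided Niagara sprite and mesh particle material for the jet flame on the BoneMarrow glider.
# A stretched, concave-warped sphere mask is multiplied by a U-axis gradient, eroded by subtracting panning wispy noise, then sharpened with SmoothStep.
# The result is multiplied by particle alpha and a view-angle fade, while particle color feeds the emissive adjustments function.
# A dynamic parameter lets particles drive the noise pan speed.
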Material /BRCosmetics/Gadgets/Assets/VinderTech_GliderChute/Glider_BoneMarrow/FX/M_Glider_BoneMarrow_JetFlame
graph {
 properties {
  bUsedWithNiagaraMeshParticles=true
  bUsedWithNiagaraSprites=true
  blendMode=BLEND_Translucent
  pixelDepthOffsetMode=PDOM_Legacy
  shadingModel=MSM_Unlit
  twoSided=true
  usedShadingModels=Unlit
 }
 nodes [
  n1 = ParticleColor()
  MaterialFunctionCall(Emissive=@n1.RGB) {materialFunction=&MF_FX_EmissiveAdjustments}
  n2 = TextureCoordinate()
  n3 = AppendVector(A=OneMinus(None=ComponentMask(None=@n2) {r=true}), B=ComponentMask(None=@n2) {g=true})
  n4 = AppendVector(A=ScalarParameter() {defaultValue=0.1, parameterName=Stretch X, sliderMax=1}, B=Constant() {r=1})
  n5 = TextureCoordinate()
  n6 = Power(Base=OneMinus(None=ComponentMask(None=@n5) {r=true}), Exp=ScalarParameter() {defaultValue=2, parameterName=Concave, sliderMax=5}) {constExponent=1.5}
  n7 = Constant() {r=1}
  n8 = SphereMask(
   A=Multiply(A=Multiply(A=@n3, B=@n4) {constB=0.1}, B=AppendVector(A=@n6, B=@n7))
   B=AppendVector(A=ScalarParameter() {defaultValue=0.5, parameterName=Center X, sliderMax=1}, B=Constant() {r=0.5})
  ) {attenuationRadius=0.5, hardnessPercent=0}
  n9 = ComponentMask(None=OneMinus(None=TextureCoordinate())) {r=true}
  n10 = Multiply(A=SmoothStep(Value=@n9) {constMax=0.75}, B=ScalarParameter() {defaultValue=5, parameterName=Shape Intensity, sliderMax=10}) {constB=5}
  n11 = TextureCoordinate()
  n12 = Constant()
  n13 = DynamicParameter() {defaultValue=<rgba 0 1 1 1>, paramNames [Pan Speed, Param2, Param3, Param4]}
  n14 = Multiply(
   A=Panner(
    Coordinate=MaterialFunctionCall(XY=@n11) {materialFunction=&Swizzle}.YX
    Speed=AppendVector(A=@n12, B=@n13.Pan Speed)
   ) {speedY=-3}
   B=AppendVector(
    A=ScalarParameter() {defaultValue=1, parameterName=Scale X, sliderMax=2}
    B=ScalarParameter() {defaultValue=1, parameterName=Scale Y, sliderMax=2}
   )
  )
  n15 = SmoothStep(
   Max=ScalarParameter() {defaultValue=1, parameterName=Hard Edge, sliderMax=1}
   Value=Subtract(
    A=Multiply(A=@n8, B=@n10) {constB=2}
    B=TextureSampleParameter2D(UVs=@n14) {parameterName=Noise Texture, texture=/BRCosmetics/Athena/Effects/Particles/Skydiving/NS_Contrail_Chainmail/T_WispyFireNoise03}.R
   )
  )
  n16 = ScalarParameter() {defaultValue=8, parameterName="FadeSensitivity (optional) (S)", sliderMax=16}
  Multiply(A=Multiply(
    A=Multiply(A=@n1.A, B=@n15)
    B=MaterialFunctionCall("FadeSensitivity (optional)"=@n16) {materialFunction=&FadeBasedOnViewAngle}.Result
   ))
  n17 = TextureCoordinate()
  OneMinus(None=@n17)
  Power(Base=@n9, Exp=ScalarParameter() {defaultValue=2, parameterName=Shape Power, sliderMax=10})
  SmoothStep(Value=ComponentMask(None=@n17) {r=true}) {constMax=0.2}
 ]
}
