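# Translucent, unlit, two-sided Niagara material for the jet flame bowl effect on the BoneMarrow glider.
# A shaped V-axis gradient minus panning, axis-swapped noise is smoothstepped into the flame mask, with pan speed driven by a dynamic parameter.
# That mask is multiplied by particle alpha and a view-based fade from camera direction against the instance X axis.
# Particle color feeds an emissive adjustment function.
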Material /BRCosmetics/Gadgets/Assets/VinderTech_GliderChute/Glider_BoneMarrow/FX/M_Glider_BoneMarrow_JetFlameBowl
graph {
 properties {
  bUsedWithNiagaraMeshParticles=true
  bUsedWithNiagaraSprites=true
  blendMode=BLEND_Translucent
  pixelDepthOffsetMode=PDOM_Legacy
  shadingModel=MSM_Unlit
  twoSided=true
  usedShadingModels=Unlit
 }
 nodes [
  n1 = DotProduct(A=CameraVectorWS(), B=Transform(None=Constant3Vector() {constant=<rgba 1 0 0 0>}) {transformSourceType=TRANSFORMSOURCE_Instance})
  Abs(None=@n1)
  n2 = ParticleColor()
  MaterialFunctionCall(Emissive=@n2.RGB) {materialFunction=&MF_FX_EmissiveAdjustments}
  n3 = ScalarParameter() {defaultValue=1, parameterName=Hard Edge, sliderMax=1}
  n4 = OneMinus(None=ComponentMask(None=TextureCoordinate()) {g=true})
  n5 = ScalarParameter() {defaultValue=2, parameterName=Concave, sliderMax=5}
  n6 = TextureCoordinate()
  n7 = SmoothStep(Value=ComponentMask(None=OneMinus(None=@n6)) {g=true}) {constMax=0.75}
  n8 = ScalarParameter() {defaultValue=5, parameterName=Shape Intensity, sliderMax=10}
  n9 = Panner(
   Coordinate=MaterialFunctionCall(XY=TextureCoordinate()) {materialFunction=&Swizzle}.YX
   Speed=AppendVector(A=DynamicParameter() {defaultValue=<rgba -1 1 1 1>, paramNames [Pan Speed, Param2, Param3, Param4]}.Pan Speed, B=Constant())
  ) {speedY=-3}
  n10 = AppendVector(
   A=ScalarParameter() {defaultValue=1, parameterName=Scale X, sliderMax=2}
   B=ScalarParameter() {defaultValue=1, parameterName=Scale Y, sliderMax=2}
  )
  n11 = Subtract(
   A=Multiply(A=Power(Base=@n4, Exp=@n5) {constExponent=1.5}, B=Multiply(A=@n7, B=@n8) {constB=5}) {constB=2}
   B=TextureSampleParameter2D(UVs=Multiply(A=@n9, B=@n10)) {parameterName=Noise Texture, texture=/BRCosmetics/Athena/Effects/Particles/Skydiving/NS_Contrail_Chainmail/T_WispyFireNoise03}.R
  )
  n12 = Constant3Vector() {constant=<rgba 1 0 0 0>}
  n13 = Abs(None=DotProduct(
    A=MaterialFunctionCall() {materialFunction=&CameraDirectionVector}.Result
    B=Transform(None=@n12) {transformSourceType=TRANSFORMSOURCE_Instance}
   ))
  n14 = ScalarParameter() {defaultValue=8, parameterName=ViewBasedFadeSensitivity, sliderMax=16}
  Multiply(A=Multiply(A=@n2.A, B=SmoothStep(Max=@n3, Value=@n11)), B=Clamp(None=Power(Base=@n13, Exp=@n14)))
  SmoothStep(
   Min=ScalarParameter() {defaultValue=0.2, parameterName=Fade Min, sliderMax=1}
   Max=ScalarParameter() {defaultValue=0.5, parameterName=Fade Max, sliderMax=1}
   Value=Saturate(None=@n1)
  ) {constMax=0.6, constMin=0.1}
 ]
}
