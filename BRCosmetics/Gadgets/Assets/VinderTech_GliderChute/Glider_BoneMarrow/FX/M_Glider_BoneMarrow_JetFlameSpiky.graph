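# Unlit, two-sided translucent Niagara particle material for the spiky jet flame on the Bone Marrow glider.
# Particle color blends toward a darkened version along U, feeding an emissive-adjustment function.
# A borrowed DualParadox emote spike texture, with scaled, offset, clamped U, pans along V at a dynamic-parameter Speed.
# Its green channel is multiplied by particle alpha and a view-angle fade to form the flame mask.
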Material /BRCosmetics/Gadgets/Assets/VinderTech_GliderChute/Glider_BoneMarrow/FX/M_Glider_BoneMarrow_JetFlameSpiky
graph {
 properties {
  bUsedWithNiagaraMeshParticles=true
  bUsedWithNiagaraSprites=true
  blendMode=BLEND_Translucent
  pixelDepthOffsetMode=PDOM_Legacy
  previewMesh=&/SM_Glider_BoneMarrow_FlameMesh03
  shadingModel=MSM_Unlit
  twoSided=true
  usedShadingModels=Unlit
 }
 nodes [
  Constant2Vector() {r=0.2}
  Constant2Vector() {g=1, r=1}
  n1 = ParticleColor()
  n2 = ScalarParameter() {defaultValue=0.5, parameterName=Gradient Dark, sliderMax=1}
  n3 = ScalarParameter() {parameterName=Min, sliderMax=1}
  n4 = Add(A=@n3, B=ScalarParameter() {defaultValue=0.1, parameterName=Width, sliderMax=1})
  n5 = ComponentMask(None=TextureCoordinate()) {r=true}
  MaterialFunctionCall(Emissive=LinearInterpolate(A=@n1.RGB, B=Multiply(A=@n1.RGB, B=@n2) {constB=0.75}, Alpha=SmoothStep(Min=@n3, Max=@n4, Value=@n5))) {materialFunction=&MF_FX_EmissiveAdjustments}
  n6 = ScalarParameter() {defaultValue=1, parameterName=Spike Scale, sliderMax=2}
  n7 = Constant() {r=1}
  n8 = Add(
   A=Multiply(A=TextureCoordinate(), B=AppendVector(A=@n6, B=@n7))
   B=AppendVector(A=ScalarParameter() {defaultValue=0.1, parameterName=Spike Offset, sliderMax=2}, B=Constant())
  )
  n9 = Clamp(None=ComponentMask(None=@n8) {r=true}) {maxDefault=0.99}
  n10 = ComponentMask(None=@n8) {g=true}
  n11 = Constant()
  n12 = DynamicParameter() {paramNames [Speed,Param2,Param3,Param4]}
  n13 = TextureSample!(UVs=Panner(Coordinate=AppendVector(A=@n9, B=@n10), Speed=AppendVector(A=@n11, B=@n12.Speed)) {speedY=1.5}, Tex, Level, Apply View MipBias) {constCoordinate=1, mipValueMode=TMVM_MipLevel, texture=/BRCosmetics/Effects/Fort_Effects/Emotes/DualParadox/T_Emote_DualParadox_Spikes_Small}
  n14 = StaticSwitchParameter(True=@n13.G, False=@n13.G) {parameterName=Glow?}
  n15 = ScalarParameter() {defaultValue=8, parameterName="FadeSensitivity (optional) (S)", sliderMax=16}
  Multiply(A=Multiply(
    A=Multiply(A=@n1.A, B=@n14)
    B=MaterialFunctionCall("FadeSensitivity (optional)"=@n15) {materialFunction=&FadeBasedOnViewAngle}.Result
   ))
  Power(Base=@n5, Exp=ScalarParameter() {defaultValue=1, parameterName=Gradient Power, sliderMax=5})
  n16 = TextureCoordinate() {coordinateIndex=1}
  SmoothStep(Value=OneMinus(None=ComponentMask(None=@n16) {g=true})) {constMax=0.1}
 ]
}
